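# Sound cue that plays the drop sound for the Alaska item in StaminaGameplay.
# A single wave player feeds a Modulator at default settings, with the cue's volume reduced.
# It uses weapon-drop attenuation but routes through the gliders sound class.
# A concurrency override caps it at two instances, taking the place of the referenced glider open/close concurrency set.
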SoundCue /StaminaGameplay/Gameplay/Items/Alaska/Audio/Alaska_Drop_Cue
root=Modulator_0
nodes [
 {id=Modulator_0, children [WavePlayer_1]}
 {id=WavePlayer_1, properties {soundWaveAssetPtr=&/Stamina_Alaska_Drop}}
]
properties {
 attenuationSettings=&Fort_Weapon_Drop_Att
 bOverrideConcurrency=true
 maxDistance=3000
 soundClassObject=/Game/Sounds/FortSoundClassesAndModes/Fort_Gliders
 volumeMultiplier=0.7
 concurrencyOverrides {maxCount=2}
 concurrencySet [&Athena_GliderOpenClose_Conc]
}
