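# Sound cue for the equip sound of the Alaska item, set up as a glider open/close sound.
# A modulator at the root feeds a single wave player that plays the Stamina Alaska equip wave.
# It uses the glider sound class and the Athena player glider open/close attenuation.
# It references the glider open/close concurrency set, with a concurrency override enabled that sets a small maximum count.
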SoundCue /StaminaGameplay/Gameplay/Items/Alaska/Audio/Alaska_Equip_Cue
root=Modulator_0
nodes [
 {id=Modulator_0, children [WavePlayer_1]}
 {id=WavePlayer_1, properties {soundWaveAssetPtr=&/Stamina_Alaska_Equip}}
]
properties {
 attenuationSettings=/Game/Sounds/Attenuation/Fort_Player_Glider_OpenClose_Athena_Att
 bOverrideConcurrency=true
 maxDistance=10000
 soundClassObject=/Game/Sounds/FortSoundClassesAndModes/Fort_Gliders
 volumeMultiplier=1
 concurrencyOverrides {maxCount=2}
 concurrencySet [&Athena_GliderOpenClose_Conc]
}
