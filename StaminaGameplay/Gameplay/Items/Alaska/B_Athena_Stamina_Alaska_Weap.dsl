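; Weapon Blueprint for the Stamina "Alaska" item, linking equip state and trigger presses to the Gameplay Ability System.
; Equipping sends an Equipped gameplay event to the instigator, and unequipping removes effects granting the Equipped tag.
; BeginPlay binds the trigger press, whose handler tries to activate GA_Stamina_Alaska.
; The handler also ends skydiving outside a vortex, except when the pawn has both sprint and falling tags.
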(event EventTick
  (params (DeltaSeconds :type Float%20%28single-precision%29))
  (|CallParentFunction DeltaSeconds) ; This node is disabled and will not be called.\nDrag off pins to build functionality.
)

(event FortWeapon|EventOnEquippedAndAttached
  ;[1 1 1] OnEquip
  (|CallParentFunction)
  (Ability|SendGameplayEventtoActor (Game|GetInstigator) "(TagName=\"Stamina.Alaska.Event.Equipped\")"))

(event FortWeapon|EventOnUnEquippedAndDetached
  ;[1 1 1] OnUnEquip
  (bind _asfort_player_pawn (Utilities|Casting|CastToFortPlayerPawn (Game|GetInstigator)))
  (GameplayEffects|RemoveActiveEffectswithGrantedTags (Ability|GetAbilitySystemComponent _asfort_player_pawn) "(GameplayTags=((TagName=\"Stamina.Alaska.Equipped\")))"))

(event EventBeginPlay
  ;[1 1 1] BeginPlay/OnTriggerPressed
  (EventDispatchers|BindEventtoOnPressTrigger (delegate Custom|OnTrigger)))

(event Custom|OnTrigger
  (params (Weapon :type Fort%20Weapon%20Object%20Reference))
  ;[1 1 1] BeginPlay/OnTriggerPressed
  (bind _asfort_player_pawn (Utilities|Casting|CastToFortPlayerPawn (Game|GetInstigator)))
  (if (and (GameplayTags|HasMatchingGameplayTag _asfort_player_pawn (Variables|Default|GetSprintTag)) (GameplayTags|HasMatchingGameplayTag _asfort_player_pawn (Variables|Default|GetFallingTag)))
    _
    (elif (and (Parachute|IsSkydiving _asfort_player_pawn) (not (Vortex|IsinVortex _asfort_player_pawn)))
      (Parachute|EndSkydiving _asfort_player_pawn)))
  (Abilities|TryActivateAbilitybyClass (Ability|GetAbilitySystemComponent _asfort_player_pawn) "/StaminaGameplay/Gameplay/Items/Alaska/GA_Stamina_Alaska.GA_Stamina_Alaska_C" true))

(fn ConstructionScript
  (|CallParentFunction))
