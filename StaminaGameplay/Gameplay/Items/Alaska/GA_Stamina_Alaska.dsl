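; GA_Stamina_Alaska is a gameplay ability for the Alaska item in StaminaGameplay that applies the item's active effect to its owner.
; On activation, it ends ziplining if the avatar Fort player pawn is both ziplining and locally controlled.
; It then calls the parent activation logic and applies GE_Stamina_Alaska_Active, leaving the rest of the item's behavior to that parent and effect.
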(event EventOnAbilityActivated
  (bind _returnvalue (CallFunction|GetAvatarasFortPlayerPawn))
  ;[1 1 1] Start ability
  (bind _returnvalue_2 (Zipline|IsZiplining _returnvalue))
  (if (and _returnvalue_2 (Pawn|IsLocallyControlled _returnvalue))
    (Zipline|EndZiplining _returnvalue true false)
    (else
      _))
  (|CallParentFunction)
  (Ability|ApplyGameplayEffectToOwner "/StaminaGameplay/Gameplay/Items/Alaska/GE_Stamina_Alaska_Active.GE_Stamina_Alaska_Active_C"))
