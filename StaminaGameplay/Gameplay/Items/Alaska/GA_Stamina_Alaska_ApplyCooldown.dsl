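; Gameplay ability for the Stamina Alaska item that applies a timed cooldown to its owner.
; On activation it builds a spec from the Alaska TimedCooldown gameplay effect and sets its SetByCaller.Cooldown magnitude from a scalable-float cooldown registry value.
; It applies that spec to the owner and immediately ends; the end and passive-activation-failure events are empty.
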(event Ability|EventActivateAbility
  (bind _returnvalue (Ability|GameplayEffect|AssignTagSetbyCallerMagnitude (Ability|MakeOutgoingGameplayEffectSpec "/StaminaGameplay/Gameplay/Items/Alaska/GE_Stamina_Alaska_TimedCooldown.GE_Stamina_Alaska_TimedCooldown_C") "(TagName=\"SetByCaller.Cooldown\")" (ScalableFloat|GetValueatLevelwithContext (Variables|Default|GetCooldownRegistryValue))))
  (Ability|ApplyGameplayEffectSpecToOwner _returnvalue)
  (Ability|EndAbility))

(event Ability|EventOnEndAbility
  (params (bWasCancelled :type Boolean)))

(event Ability|EventFailedToActivatePassiveAbility
  (params (ActorInfo :type Gameplay%20Ability%20Actor%20Info%20Structure)))
